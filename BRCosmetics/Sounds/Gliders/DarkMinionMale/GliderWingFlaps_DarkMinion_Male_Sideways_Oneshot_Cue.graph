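# Sound cue for a one-shot sideways wing flap of the male Dark Minion glider.
# A random node picks one of four equally weighted wave variations and feeds a modulator that randomizes pitch.
# It uses glider thrust attenuation and the glider landing sound class, and primes on load.
# A concurrency override caps it at two instances per owner, stopping the oldest.
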SoundCue /BRCosmetics/Sounds/Gliders/DarkMinionMale/GliderWingFlaps_DarkMinion_Male_Sideways_Oneshot_Cue
root=Modulator_0
nodes [
 {id=Modulator_0, children [Random_0], properties {pitchMax=1.1, pitchMin=0.85, volumeMin=1}}
 {id=Random_0, children [WavePlayer_1,WavePlayer_2,WavePlayer_6,WavePlayer_7], properties {weights [1,1,1,1]}}
 {id=WavePlayer_1, properties {soundWaveAssetPtr=&/Glider_DarkMinion_Sideways_Oneshot_01}}
 {id=WavePlayer_2, properties {soundWaveAssetPtr=&/Glider_DarkMinion_Sideways_Oneshot_02}}
 {id=WavePlayer_6, properties {soundWaveAssetPtr=&/Glider_DarkMinion_Sideways_Oneshot_03}}
 {id=WavePlayer_7, properties {soundWaveAssetPtr=&/Glider_DarkMinion_Sideways_Oneshot_04}}
]
properties {
 attenuationSettings=&Fort_Player_Glider_Thrust_Athena_Att
 bOverrideConcurrency=true
 bPrimeOnLoad=true
 maxDistance=12500
 soundClassObject=&Fort_Gliders_Land
 volumeMultiplier=0.8
 concurrencyOverrides {bLimitToOwner=true, maxCount=2, resolutionRule=StopOldest}
 concurrencySet [&Athena_GliderOpenClose_Conc]
}
